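# Material function that adds glowing algae to the Hera ocean water surface's BaseColor and EmissiveColor.
# Up to three rotated, oscillating algae texture layers, reduced by quality level, form the pattern within a world-placed mask.
# Flow-distorted masks, shoreline depth and camera-faded turbulence detail blend the low and two high colors.
# A time-of-day scalar scales the result, but all four periods read the Morning channel.
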MaterialFunction /FortWater/Materials/WaterSurface/Hera/Ocean/GlowingAlgae/MF_GlowAlgae
graph {nodes [
  n1 = GetMaterialAttributes(None=FunctionInput() {bUsePreviewValueAsDefault=true, inputName=Material Attribs, inputType=FunctionInput_MaterialAttributes}) {attributeGetTypes [#1, #2]}
  n2 = VectorParameter() {group=Hera Algae, parameterName=Algae Mask Size, channelNames {a=Scale all, b="", g="", r=""}, defaultValue=<rgba 10000 10000 10000 0>}
  n3 = WorldPosition()
  n4 = VectorParameter() {group=Hera Algae, parameterName=Algae Mask Position}
  n5 = ComponentMask(None=MaterialFunctionCall!(
    Export Float 4
    TextureObject=TextureObjectParameter() {group=Hera Algae, parameterName=Algae Mask, texture=&Black}
    TextureSize=Multiply(A=@n2.RGB, B=@n2.RGB)
    WorldPosition=Subtract(A=@n3.XYZ, B=@n4.RGB)
    World Space Normal
    RotZAxis
    RotXAxis
    RotYAxis
    ProjectionTransitionContrast
   ) {materialFunction=/Clyde/Athena/Apollo/Water/Materials/MF_WorldAlignedTexture_NoWrap}.Z Texture) {r=true}
  n6 = VectorParameter() {group=Hera Algae, parameterName=Algae TOD Emissive Multiplier, channelNames {a=Night, b=Eveing, g=Day, r=Morning}}
  n7 = MaterialFunctionCall(Value_Morning=@n6.Morning, Value_Day=@n6.Morning, Value_Evening=@n6.Morning, Value_Night=@n6.Morning) {materialFunction=&Scalar_Time_of_Day}
  n8 = Multiply(A=TextureCoordinate(), B=ScalarParameter() {defaultValue=1, group=Hera Algae, parameterName=Algae scale})
  n9 = Multiply(A=Time(), B=ScalarParameter() {defaultValue=0.1, group=Hera Algae, parameterName=sine pan speed}) {constB=0.1}
  n10 = Sine(None=@n9)
  n11 = ScalarParameter() {defaultValue=0.1, group=Hera Algae, parameterName=sine pan amt}
  n12 = TextureObject() {samplerType=SAMPLERTYPE_LinearGrayscale, texture=/Game/Global/Textures/Default/PointSplatter/T_GlowAlgae}
  n13 = TextureSample(UVs=Panner(Coordinate=@n8, Time=Multiply(A=@n10, B=@n11)) {speedX=1}, Tex=@n12) {
   samplerSource=SSM_Wrap_WorldGroupSettings
   samplerType=SAMPLERTYPE_LinearGrayscale
   texture=&T_GlowAlgae
  }
  n14 = Constant() {r=0.333}
  n15 = VectorParameter() {group=Hera Algae, parameterName=2nd and third Layer options, defaultValue=<rgba 0 1 0 1>}
  n16 = Multiply(
   A=Add(
    A=MaterialFunctionCall(UVs=@n8, "Rotation Angle (0-1)"=@n14) {materialFunction=&CustomRotator}.Rotated Values
    B=@n15.R
   )
   B=@n15.G
  ) {constB=0.725}
  n17 = ScalarParameter() {group=Hera Algae, parameterName=BO Height 1}
  n18 = Cosine(None=@n9)
  n19 = TextureSample(UVs=Panner(Coordinate=BumpOffset(Coordinate=@n16, Height=@n17), Time=Multiply(A=@n18, B=@n11)) {speedX=1}, Tex=@n12) {
   samplerSource=SSM_Wrap_WorldGroupSettings
   samplerType=SAMPLERTYPE_LinearGrayscale
   texture=&T_GlowAlgae
  }
  n20 = Min(A=@n13.R, B=@n19.R)
  n21 = Constant() {r=0.666}
  n22 = Multiply(
   A=Add(
    A=MaterialFunctionCall(UVs=@n8, "Rotation Angle (0-1)"=@n21) {materialFunction=&CustomRotator}.Rotated Values
    B=@n15.B
   ) {constB=2}
   B=@n15.A
  ) {constB=1.3}
  n23 = ScalarParameter() {group=Hera Algae, parameterName=BO Height 2}
  n24 = Sine(None=Add(A=@n9) {constB=0.35})
  n25 = TextureSample(UVs=Panner(Coordinate=BumpOffset(Coordinate=@n22, Height=@n23), Time=Multiply(A=@n24, B=@n11)) {speedX=1}) {
   samplerSource=SSM_Wrap_WorldGroupSettings
   samplerType=SAMPLERTYPE_LinearGrayscale
   texture=&T_GlowAlgae
  }
  n26 = Max(A=@n13.R, B=@n19.R)
  n27 = ScalarParameter() {defaultValue=1, group=Hera Algae, parameterName=Min Vs Max}
  n28 = LinearInterpolate(A=Min(A=@n20, B=@n25.R), B=Max(A=@n26, B=@n25.R), Alpha=@n27)
  n29 = LinearInterpolate(A=@n20, B=@n26, Alpha=@n27)
  n30 = QualitySwitch(Default=@n28, Low=@n13.R, High=@n29, Medium=@n29, Epic=@n28)
  n31 = VectorParameter() {group=Hera Algae, parameterName=algae texture options, defaultValue=<rgba 1 1 0 0>}
  n32 = Add(A=Multiply(A=Power(Base=@n30, Exp=@n31.R), B=@n31.G), B=@n31.B)
  n33 = VectorParameter() {group=Hera Algae, parameterName=Algae Color High 0, defaultValue=<rgba 0.0052 0.685 1 1>}
  n34 = VectorParameter() {group=Hera Algae, parameterName=Algae Color High 1, defaultValue=<rgba 0 1 0.3167 1>}
  n35 = WorldPosition()
  n36 = ScalarParameter() {defaultValue=0.0001, group=Hera Algae, parameterName=Flow Map Alpha Texture Scale}
  n37 = ComponentMask(None=NamedRerouteUsage()) {g=true, r=true}
  n38 = NamedRerouteDeclaration(None=Multiply(A=Multiply(A=@n37) {constB=-1}, B=NamedRerouteUsage())) {nodeColor=<rgba 0 0.4844 1 1>}
  n39 = ScalarParameter() {defaultValue=0.2, group=Hera Algae, parameterName=Alpha Wave Distortion Strength}
  n40 = NamedRerouteUsage()
  n41 = ScalarParameter() {defaultValue=0.01, group=Hera Algae, parameterName=Alpha Texture Speed}
  n42 = Panner(
   Coordinate=Add(A=Multiply(A=@n35.XY, B=@n36), B=Multiply(A=@n38, B=@n39))
   Time=Time()
   Speed=Multiply(A=Multiply(A=@n40, B=@n41)) {constB=-1}
  ) {bFractionalPart=true}
  n43 = WorldPosition()
  n44 = ScalarParameter() {defaultValue=0, group=Hera Algae, parameterName=Flow Map Alpha Distortion Scale}
  n45 = Divide(A=@n43.XY, B=@n44)
  n46 = Time()
  n47 = Multiply(A=Multiply(A=NamedRerouteUsage(), B=ScalarParameter() {group=Hera Algae, parameterName=Flow Map Glow Alpha Motion Speed})) {constB=-1}
  n48 = ComponentMask(None=TextureSample(
    UVs=Panner(Coordinate=@n45, Time=@n46, Speed=@n47) {bFractionalPart=true}
    Tex=TextureObject() {samplerType=SAMPLERTYPE_LinearColor, texture=/Game/Packages/Fortnite_UniqueMaterials_1/Textures/FlowMap}
   ) {samplerSource=SSM_Wrap_WorldGroupSettings, samplerType=SAMPLERTYPE_LinearColor}.RGB) {g=true, r=true}
  n49 = Add(
   A=Multiply(
    A=ConstantBiasScale(None=@n48) {bias=-0.5, scale=2}
    B=ScalarParameter() {defaultValue=0.4, group=Hera Algae, parameterName=Flow Map Alpha Strength}
   )
   B=Multiply(A=Time(), B=ScalarParameter() {defaultValue=0.005, group=Hera Algae, parameterName=Flow Map Alpha Speed})
  )
  n50 = ComponentMask(None=TextureSample(
    UVs=Add(A=@n42, B=@n49)
    Tex=TextureObject() {samplerType=SAMPLERTYPE_Masks, texture=/Game/Global/Textures/Default/Masks_and_Gradients/Swamp_Mask}
   ) {samplerSource=SSM_Wrap_WorldGroupSettings}.RGB) {g=true}
  n51 = ScalarParameter() {defaultValue=2, group=Hera Algae, parameterName=Alpha Color Power}
  n52 = Saturate(None=Multiply(A=Power(Base=@n50, Exp=@n51), B=ScalarParameter() {defaultValue=1, group=Hera Algae, parameterName=Alpha Color Strength}))
  n53 = Saturate(None=NamedRerouteUsage())
  n54 = Divide(
   A=FunctionInput(Preview=MaterialFunctionCall!(Texture, Position) {materialFunction=/FortWater/Materials/Functions/WaterBodyData}.Water Depth) {bUsePreviewValueAsDefault=true, inputName=Water Depth, inputType=FunctionInput_Scalar, sortPriority=1}
   B=ScalarParameter() {defaultValue=1200, group=Hera Algae, parameterName=Water Shoreline Depth}
  )
  n55 = Power(Base=NamedRerouteDeclaration(None=Saturate(None=@n54)) {name=shore dpeth, nodeColor=<rgba 0 1 0.5547 1>})
  n56 = NamedRerouteUsage()
  n57 = ScalarParameter() {defaultValue=0.05, group=Hera Algae, parameterName=Flow Map Glow Distortion Motion Speed}
  n58 = Panner(
   Coordinate=Multiply(A=WorldPosition().XY, B=ScalarParameter() {defaultValue=0, group=Hera Algae, parameterName=Flow Map Glow Distortion Scale})
   Time=Time()
   Speed=Multiply(A=Multiply(A=@n56, B=@n57)) {constB=-1}
  )
  n59 = TextureObject() {samplerType=SAMPLERTYPE_LinearColor, texture=&FlowMap}
  n60 = ConstantBiasScale(None=ComponentMask(None=TextureSample(UVs=@n58, Tex=@n59) {samplerSource=SSM_Wrap_WorldGroupSettings, samplerType=SAMPLERTYPE_LinearColor}.RGB) {g=true, r=true}) {bias=-0.5, scale=2}
  n61 = ScalarParameter() {defaultValue=1, group=Hera Algae, parameterName=Flow Map Glow Strength}
  n62 = Time()
  n63 = ScalarParameter() {defaultValue=0.05, group=Hera Algae, parameterName=Flow Map Glow Speed}
  n64 = WorldPosition()
  n65 = ScalarParameter() {defaultValue=0.0002, group=Hera Algae, parameterName=Flow Map Glow Texture Scale}
  n66 = NamedRerouteUsage()
  n67 = ScalarParameter() {defaultValue=0.2, group=Hera Algae, parameterName=Glow Wave Distortion Strength}
  n68 = Add(A=Add(A=Multiply(A=@n60, B=@n61), B=Multiply(A=@n62, B=@n63)), B=Add(A=Multiply(A=@n64.XY, B=@n65), B=Multiply(A=@n66, B=@n67)))
  n69 = TextureObject() {samplerType=SAMPLERTYPE_Masks, texture=&Swamp_Mask}
  n70 = Power(
   Base=ComponentMask(None=TextureSample(UVs=@n68, Tex=@n69) {samplerSource=SSM_Wrap_WorldGroupSettings}.RGB) {b=true}
   Exp=LinearInterpolate(
    A=ScalarParameter() {defaultValue=3, group=Hera Algae, parameterName=Glow Power Min}
    B=ScalarParameter() {defaultValue=1, group=Hera Algae, parameterName=Glow Power Max}
    Alpha=LinearInterpolate(B=@n53, Alpha=@n55) {constA=1}
   )
  )
  n71 = Add(
   A=Multiply(A=WorldPosition().XY, B=ScalarParameter() {defaultValue=0.0035, group=Hera Algae, parameterName=Accent Texture Scale})
   B=Multiply(A=NamedRerouteUsage(), B=ScalarParameter() {defaultValue=2, group=Hera Algae, parameterName=Accent Wave Distortion Strength})
  ) {constB=0}
  n72 = Time()
  n73 = Normalize(VectorInput=VectorParameter() {group=Hera Algae, parameterName=Wave Direction, defaultValue=<rgba 0.5 0.5 0 1>}.RGB)
  n74 = Multiply(A=Multiply(
    A=NamedRerouteDeclaration(None=@n73) {name=Wave Dir, nodeColor=<rgba 1 0 0.1641 1>}
    B=ScalarParameter() {defaultValue=0.1, group=Hera Algae, parameterName=Accent Texture Speed}
   )) {constB=-1}
  n75 = TextureObject() {texture=/FortWater/Textures/T_Turbulence_Seamless}
  n76 = ComponentMask(None=TextureSample(UVs=Panner(Coordinate=@n71, Time=@n72, Speed=@n74), Tex=@n75) {samplerSource=SSM_Wrap_WorldGroupSettings, samplerType=SAMPLERTYPE_Masks, texture=/FortWater/Textures/T_OceanFoam_New_Mask}.RGB) {r=true}
  n77 = Add(
   A=Multiply(A=WorldPosition().XY, B=ScalarParameter() {defaultValue=0.0025, group=Hera Algae, parameterName=Accent Texture Scale 2})
   B=Multiply(A=NamedRerouteUsage(), B=ScalarParameter() {defaultValue=2, group=Hera Algae, parameterName=Accent Wave Distortion Strength 2})
  ) {constB=0}
  n78 = Multiply(A=Multiply(A=NamedRerouteUsage(), B=ScalarParameter() {defaultValue=0.05, group=Slurp, parameterName=Accent Texture Speed 2})) {constB=-1}
  n79 = ComponentMask(None=TextureSample(UVs=Panner(Coordinate=@n77, Time=@n72, Speed=@n78), Tex=@n75) {samplerSource=SSM_Wrap_WorldGroupSettings, samplerType=SAMPLERTYPE_Masks, texture=&T_OceanFoam_New_Mask}.RGB) {r=true}
  n80 = NamedRerouteDeclaration(None=Saturate(None=LinearInterpolate(A=@n76, B=@n79) {constAlpha=0.75})) {name=color mask, nodeColor=<rgba 0.875 0 1 1>}
  n81 = ScalarParameter() {defaultValue=0.5, group=Hera Algae, parameterName=Detail Power}
  n82 = ScalarParameter() {defaultValue=5000, group=Hera Algae, parameterName=Camera Effects Fade Length}
  n83 = ScalarParameter() {defaultValue=1000, group=Hera Algae, parameterName=Camera Effects Fade Offset}
  n84 = StaticBool() {value=true}
  n85 = LinearInterpolate(
   A=LinearInterpolate(
    A=ScalarParameter() {defaultValue=0.6, group=Hera Algae, parameterName=Detail Lerp Low}
    B=ScalarParameter() {defaultValue=1.75, group=Hera Algae, parameterName=Detail Lerp High}
    Alpha=Power(Base=@n80, Exp=@n81)
   ) {constA=0.5}
   Alpha=NamedRerouteDeclaration(None=MaterialFunctionCall(Fade Length=@n82, Fade Offset=@n83, For Vertex Shader=@n84) {materialFunction=&CameraDepthFade}.Result) {name=Cam Fade, nodeColor=<rgba 1 0.2344 0 1>}
  )
  n86 = Multiply(
   A=Multiply(
    A=Multiply(A=@n5, B=@n7.Result)
    B=MaterialFunctionCall(In=@n32, Contrast=@n31.A) {materialFunction=&CheapContrast}.Result
   )
   B=LinearInterpolate(
    A=VectorParameter() {group=Hera Algae, parameterName=Algae Color Low, defaultValue=<rgba 0 0.0521 0.1042 1>}.RGB
    B=LinearInterpolate(A=@n33.RGB, B=@n34.RGB, Alpha=@n52)
    Alpha=Multiply(A=@n70, B=@n85)
   )
  )
  n87 = Multiply(A=ScalarParameter() {group=Hera Algae, parameterName=Algae Base color Mul}, B=@n86)
  n88 = Multiply(A=@n86, B=ScalarParameter() {group=Hera Algae, parameterName=Algae Emissive Mul})
  FunctionOutput(None=SetMaterialAttributes!(MaterialAttributes=@n1.MaterialAttributes, BaseColor=Add(A=@n1.BaseColor, B=@n87), EmissiveColor=Add(A=@n1.EmissiveColor, B=@n88) {constB=0}) {attributeSetTypes [#1, #2]})
  Multiply(A=@n43.XY, B=@n44)
  n89 = MaterialFunctionCall() {materialFunction=&LinearGradient}
  n90 = OneMinus(None=Power(Base=OneMinus(None=@n89.UGradient)) {constExponent=2.2})
  n91 = OneMinus(None=ConstantBiasScale(None=Cosine(None=@n90)))
  n92 = ScalarParameter() {defaultValue=0.2, group=Hera Algae, parameterName=Wave Shoreline Dampening Width}
  n93 = NamedRerouteUsage()
  n94 = Power(
   Base=Saturate(None=InverseLinearInterpolate(B=@n92, Value=@n93))
   Exp=ScalarParameter() {defaultValue=0.8, group=Hera Algae, parameterName=Wave Shoreline Dampening Power}
  )
  n95 = InverseLinearInterpolate(
   B=ScalarParameter() {defaultValue=0.8, group=Hera Algae, parameterName=Wave Sphere Cutoff High}
   Value=SphereMask(
    A=WorldPosition().XYZ
    B=VectorParameter() {group=Hera Algae, parameterName=Wave Sphere Center, defaultValue=<rgba -35410 -110415 0 1>}.RGB
    Radius=ScalarParameter() {defaultValue=17500, group=Hera Algae, parameterName=Wave Sphere Radius}
   ) {hardnessPercent=0}
  )
  NamedRerouteDeclaration(None=Multiply(A=Multiply(A=@n91, B=@n94), B=Saturate(None=@n95))) {nodeColor=<rgba 0.8047 1 0 1>}
 ]}
